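# Control Rig for the AbstractMirror backpack's small mech, procedurally driving its left and right leg pistons.
# Each lower piston follows its foot, and each pelvis piston its thigh, via X-translation-only parent constraints, then both halves aim at each other.
# A separate brake chain, unconnected to execution, would set brake bones from toe local transforms, using frame-to-frame rotation deltas with yaw scaled down.
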ControlRigBlueprint /BRCosmetics/Accessories/FORT_Backpacks/Backpack_M_MED_AbstractMirror/Meshes/M_MED_AbstractMirror_Pack_CtrlRig
graphs [
 CollapseNode_ContainedGraph {nodes [
   DeltaFromPrevious = /Script/RigVM.RigVMFunction_DeltaFromPreviousQuat(Value=RigVMFunctionEntryNode().Value) {is=struct RigVMUnitNode, template="DeltaFromPrevious::Execute(in Value,out Delta,out PreviousValue)"}
   n1 = RigVMDispatchNode(Struct=/Script/RigVM.RigVMFunction_MathQuaternionToRotator(Value=@DeltaFromPrevious.Delta) {is=struct RigVMUnitNode, template="Cast::Execute(in Value,out Result)"}.Result) {template="DISPATCH_RigVMDispatch_BreakStruct(in Struct,out Elements)"}
   DISPATCH_RigVMDispatch_MakeStruct = RigVMDispatchNode(
    Elements=(Pitch=0,Yaw=0,Roll=0)
    Elements.Pitch=@n1.Elements.Pitch
    Elements.Roll=@n1.Elements.Roll
    Elements.Yaw=&RigVMFunction_MathDoubleRemap(SourceMaximum=60, SourceMinimum=-60, TargetMaximum=15, TargetMinimum=-15, Value=@n1.Elements.Yaw).Result
   ) {template="DISPATCH_RigVMDispatch_MakeStruct(in Elements,out Struct)"}
   RigVMFunction_MathQuaternionFromRotatorV2 = /Script/RigVM.RigVMFunction_MathQuaternionFromRotatorV2(Value=@DISPATCH_RigVMDispatch_MakeStruct.Struct) {is=struct RigVMUnitNode, template="Cast::Execute(in Value,out Result)"}
   RigVMFunctionReturnNode(Result=@RigVMFunction_MathQuaternionFromRotatorV2.Result)
  ]}
 CollapseNode_ContainedGraph {nodes [
   DeltaFromPrevious = &RigVMFunction_DeltaFromPreviousQuat(Value=RigVMFunctionEntryNode().Value) {template="DeltaFromPrevious::Execute(in Value,out Delta,out PreviousValue)"}
   n1 = RigVMDispatchNode(Struct=&RigVMFunction_MathQuaternionToRotator(Value=@DeltaFromPrevious.Delta) {template="Cast::Execute(in Value,out Result)"}.Result) {template="DISPATCH_RigVMDispatch_BreakStruct(in Struct,out Elements)"}
   DISPATCH_RigVMDispatch_MakeStruct = RigVMDispatchNode(
    Elements=(Pitch=0,Yaw=0,Roll=0)
    Elements.Pitch=@n1.Elements.Pitch
    Elements.Roll=@n1.Elements.Roll
    Elements.Yaw=&RigVMFunction_MathDoubleRemap(SourceMaximum=60, SourceMinimum=-60, TargetMaximum=15, TargetMinimum=-15, Value=@n1.Elements.Yaw).Result
   ) {template="DISPATCH_RigVMDispatch_MakeStruct(in Elements,out Struct)"}
   RigVMFunction_MathQuaternionFromRotatorV2 = &RigVMFunction_MathQuaternionFromRotatorV2(Value=@DISPATCH_RigVMDispatch_MakeStruct.Struct) {template="Cast::Execute(in Value,out Result)"}
   RigVMFunctionReturnNode(Result=@RigVMFunction_MathQuaternionFromRotatorV2.Result)
  ]}
 RigVMFunctionLibrary {links [], nodes []}
 RigVMModel {nodes [
   RigUnit_BeginExecution = &RigUnit_BeginExecution()
   &RigUnit_ParentConstraint(
    AdvancedSettings=(InterpolationType=Average,RotationOrderForFilter=XZY)
    Child=(Type=Bone,Name="dyn_mech_piston02_l")
    Filter=(TranslationFilter=(bX=true,bY=false,bZ=false),RotationFilter=(bX=false,bY=false,bZ=false),ScaleFilter=(bX=false,bY=false,bZ=false))
    Parents=((Item=(Type=Bone,Name="dyn_mech_foot_l"),Weight=1))
    Weight=1
    bMaintainOffset=true
    ExecutePin=@RigUnit_BeginExecution.ExecutePin
   )
   > &RigUnit_AimConstraintLocalSpaceOffset(
    AdvancedSettings=(DebugSettings=(bEnabled=false,Scale=10,WorldOffset=(Rotation=(X=0,Y=0,Z=0,W=1),Translation=(X=0,Y=0,Z=0),Scale3D=(X=1,Y=1,Z=1))),RotationOrderForFilter=XYZ)
    Child=(Type=Bone,Name="dyn_mech_piston02_l")
    Filter=(bX=true,bY=false,bZ=false)
    Parents=((Item=(Type=Bone,Name="dyn_mech_piston01_l"),Weight=1))
    Weight=1
    WorldUp=(Target=(X=0,Y=0,Z=1),Kind=Direction,Space=(Type=None,Name="None"))
    bMaintainOffset=true
   )
   > &RigUnit_AimConstraintLocalSpaceOffset(
    AdvancedSettings=(DebugSettings=(bEnabled=false,Scale=10,WorldOffset=(Rotation=(X=0,Y=0,Z=0,W=1),Translation=(X=0,Y=0,Z=0),Scale3D=(X=1,Y=1,Z=1))),RotationOrderForFilter=XYZ)
    Child=(Type=Bone,Name="dyn_mech_piston01_l")
    Filter=(bX=true,bY=false,bZ=false)
    Parents=((Item=(Type=Bone,Name="dyn_mech_piston02_l"),Weight=1))
    Weight=1
    WorldUp=(Target=(X=0,Y=0,Z=1),Kind=Direction,Space=(Type=None,Name="None"))
    bMaintainOffset=true
   )
   > &RigUnit_ParentConstraint(
    AdvancedSettings=(InterpolationType=Average,RotationOrderForFilter=XZY)
    Child=(Type=Bone,Name="dyn_mech_piston02_r")
    Filter=(TranslationFilter=(bX=true,bY=false,bZ=false),RotationFilter=(bX=false,bY=false,bZ=false),ScaleFilter=(bX=false,bY=false,bZ=false))
    Parents=((Item=(Type=Bone,Name="dyn_mech_foot_r"),Weight=1))
    Weight=1
    bMaintainOffset=true
   )
   > &RigUnit_AimConstraintLocalSpaceOffset(
    AdvancedSettings=(DebugSettings=(bEnabled=false,Scale=10,WorldOffset=(Rotation=(X=0,Y=0,Z=0,W=1),Translation=(X=0,Y=0,Z=0),Scale3D=(X=1,Y=1,Z=1))),RotationOrderForFilter=XYZ)
    Child=(Type=Bone,Name="dyn_mech_piston02_r")
    Filter=(bX=true,bY=false,bZ=false)
    Parents=((Item=(Type=Bone,Name="dyn_mech_piston01_r"),Weight=1))
    Weight=1
    WorldUp=(Target=(X=0,Y=0,Z=1),Kind=Direction,Space=(Type=None,Name="None"))
    bMaintainOffset=true
   )
   > &RigUnit_AimConstraintLocalSpaceOffset(
    AdvancedSettings=(DebugSettings=(bEnabled=false,Scale=10,WorldOffset=(Rotation=(X=0,Y=0,Z=0,W=1),Translation=(X=0,Y=0,Z=0),Scale3D=(X=1,Y=1,Z=1))),RotationOrderForFilter=XYZ)
    Child=(Type=Bone,Name="dyn_mech_piston01_r")
    Filter=(bX=true,bY=false,bZ=false)
    Parents=((Item=(Type=Bone,Name="dyn_mech_piston02_r"),Weight=1))
    Weight=1
    WorldUp=(Target=(X=0,Y=0,Z=1),Kind=Direction,Space=(Type=None,Name="None"))
    bMaintainOffset=true
   )
   > &RigUnit_ParentConstraint(
    AdvancedSettings=(InterpolationType=Average,RotationOrderForFilter=XZY)
    Child=(Type=Bone,Name="dyn_mech_pelvis_piston_02_l")
    Filter=(TranslationFilter=(bX=true,bY=false,bZ=false),RotationFilter=(bX=false,bY=false,bZ=false),ScaleFilter=(bX=false,bY=false,bZ=false))
    Parents=((Item=(Type=Bone,Name="dyn_mech_thigh_l"),Weight=1))
    Weight=1
    bMaintainOffset=true
   )
   > &RigUnit_AimConstraintLocalSpaceOffset(
    AdvancedSettings=(DebugSettings=(bEnabled=false,Scale=10,WorldOffset=(Rotation=(X=0,Y=0,Z=0,W=1),Translation=(X=0,Y=0,Z=0),Scale3D=(X=1,Y=1,Z=1))),RotationOrderForFilter=XYZ)
    Child=(Type=Bone,Name="dyn_mech_pelvis_piston_02_l")
    Filter=(bX=true,bY=false,bZ=false)
    Parents=((Item=(Type=Bone,Name="dyn_mech_pelvis_piston_01_l"),Weight=1))
    Weight=1
    WorldUp=(Target=(X=0,Y=0,Z=1),Kind=Direction,Space=(Type=None,Name="None"))
    bMaintainOffset=true
   )
   > &RigUnit_AimConstraintLocalSpaceOffset(
    AdvancedSettings=(DebugSettings=(bEnabled=false,Scale=10,WorldOffset=(Rotation=(X=0,Y=0,Z=0,W=1),Translation=(X=0,Y=0,Z=0),Scale3D=(X=1,Y=1,Z=1))),RotationOrderForFilter=XYZ)
    Child=(Type=Bone,Name="dyn_mech_pelvis_piston_01_l")
    Filter=(bX=true,bY=false,bZ=false)
    Parents=((Item=(Type=Bone,Name="dyn_mech_pelvis_piston_02_l"),Weight=1))
    Weight=1
    WorldUp=(Target=(X=0,Y=0,Z=1),Kind=Direction,Space=(Type=None,Name="None"))
    bMaintainOffset=true
   )
   > &RigUnit_ParentConstraint(
    AdvancedSettings=(InterpolationType=Average,RotationOrderForFilter=XZY)
    Child=(Type=Bone,Name="dyn_mech_pelvis_piston_02_r")
    Filter=(TranslationFilter=(bX=true,bY=false,bZ=false),RotationFilter=(bX=false,bY=false,bZ=false),ScaleFilter=(bX=false,bY=false,bZ=false))
    Parents=((Item=(Type=Bone,Name="dyn_mech_thigh_r"),Weight=1))
    Weight=1
    bMaintainOffset=true
   )
   > &RigUnit_AimConstraintLocalSpaceOffset(
    AdvancedSettings=(DebugSettings=(bEnabled=false,Scale=10,WorldOffset=(Rotation=(X=0,Y=0,Z=0,W=1),Translation=(X=0,Y=0,Z=0),Scale3D=(X=1,Y=1,Z=1))),RotationOrderForFilter=XYZ)
    Child=(Type=Bone,Name="dyn_mech_pelvis_piston_02_r")
    Filter=(bX=true,bY=false,bZ=false)
    Parents=((Item=(Type=Bone,Name="dyn_mech_pelvis_piston_01_r"),Weight=1))
    Weight=1
    WorldUp=(Target=(X=0,Y=0,Z=1),Kind=Direction,Space=(Type=None,Name="None"))
    bMaintainOffset=true
   )
   > &RigUnit_AimConstraintLocalSpaceOffset(
    AdvancedSettings=(DebugSettings=(bEnabled=false,Scale=10,WorldOffset=(Rotation=(X=0,Y=0,Z=0,W=1),Translation=(X=0,Y=0,Z=0),Scale3D=(X=1,Y=1,Z=1))),RotationOrderForFilter=XYZ)
    Child=(Type=Bone,Name="dyn_mech_pelvis_piston_01_r")
    Filter=(bX=true,bY=false,bZ=false)
    Parents=((Item=(Type=Bone,Name="dyn_mech_pelvis_piston_02_r"),Weight=1))
    Weight=1
    WorldUp=(Target=(X=0,Y=0,Z=1),Kind=Direction,Space=(Type=None,Name="None"))
    bMaintainOffset=true
   )
   n1 = &RigUnit_GetTransform(Item=(Type=Bone,Name="dyn_mech_toe_l"), Space=LocalSpace, bInitial=False)
   BrakeAngleFromDeltaLeft = RigVMCollapseNode(Value=@n1.Transform.Rotation)
   &RigUnit_SetTransform(
    ExecutePin=()
    Item=(Type=Bone,Name="dyn_mech_brake_l")
    Space=LocalSpace
    Value=(Rotation=(X=0,Y=0,Z=0,W=1),Translation=(X=0,Y=0,Z=0),Scale3D=(X=1,Y=1,Z=1))
    Weight=1
    bInitial=false
    bPropagateToChildren=false
    Value.Rotation=@BrakeAngleFromDeltaLeft.Result
    Value.Scale3D=@n1.Transform.Scale3D
    Value.Translation=@n1.Transform.Translation
   )
   n2 = &RigUnit_GetTransform(Item=(Type=Bone,Name="dyn_mech_toe_r"), Space=LocalSpace, bInitial=False)
   BrakeAngleFromDeltaRight = RigVMCollapseNode(Value=@n2.Transform.Rotation)
   > &RigUnit_SetTransform(
    Item=(Type=Bone,Name="dyn_mech_brake_r")
    Space=LocalSpace
    Value=(Rotation=(X=0,Y=0,Z=0,W=1),Translation=(X=0,Y=0,Z=0),Scale3D=(X=1,Y=1,Z=1))
    Weight=1
    bInitial=false
    bPropagateToChildren=false
    Value.Rotation=@BrakeAngleFromDeltaRight.Result
    Value.Scale3D=@n2.Transform.Scale3D
    Value.Translation=@n2.Transform.Translation
   )
  ]}
]
hierarchy [
 root {loc=<0 0 0>, rot=<0 0 0 1>}
 pack {parent=root, loc=<0 -0.7316 0>, rot=<0 0 0 1>}
 body {parent=pack, loc=<0 -5.4011 -0.7809>, rot=<0 0 0 1>}
 mech_spine {parent=body, loc=<0 0.0012 0.0023>, rot=<0 -0.7071 0 -0.7071>}
 dyn_mech_pelvis_l {parent=mech_spine, loc=<2.2798 -0.2273 -1.9811>, rot=<-0.4455 -0.4483 0.5497 -0.5463>}
 dyn_mech_thigh_l {parent=dyn_mech_pelvis_l, loc=<3.0943 0 0>, rot=<-0.5292 -0.3074 0.6863 0.393>}
 dyn_mech_calf_l {parent=dyn_mech_thigh_l, loc=<6.4304 0 0>, rot=<0.0255 0.0045 0.6429 -0.7655>}
 dyn_mech_foot_l {parent=dyn_mech_calf_l, loc=<5.8838 0 0>, rot=<-0.0001 0 -0.6876 -0.7261>}
 dyn_mech_toe_l {parent=dyn_mech_foot_l, loc=<4.3718 0.2131 0.0046>, rot=<0 0 0 1>}
 dyn_mech_brake_l {parent=dyn_mech_foot_l, loc=<4.3718 0.2131 0.0046>, rot=<0 0 0 1>}
 dyn_mech_piston01_l {parent=dyn_mech_calf_l, loc=<2.0967 0.9094 0.0191>, rot=<-0.0028 -0.0033 0.0004 -1>}
 dyn_mech_piston02_l {parent=dyn_mech_piston01_l, loc=<3.2962 0 0>, rot=<0 0 0 1>}
 dyn_mech_pelvis_r {parent=mech_spine, loc=<2.2798 -0.2273 1.9811>, rot=<0.4483 -0.4455 -0.5463 -0.5497>}
 dyn_mech_thigh_r {parent=dyn_mech_pelvis_r, loc=<-3.0943 0 0>, rot=<-0.5292 -0.3074 0.6863 0.393>}
 dyn_mech_calf_r {parent=dyn_mech_thigh_r, loc=<-6.4304 0 0>, rot=<0.0255 0.0045 0.6429 -0.7655>}
 dyn_mech_foot_r {parent=dyn_mech_calf_r, loc=<-5.8838 0 0>, rot=<-0.0001 0 -0.6876 -0.7261>}
 dyn_mech_toe_r {parent=dyn_mech_foot_r, loc=<-4.3392 -0.2253 -0.0656>, rot=<0.001 0.0023 -0.0007 -1>}
 dyn_mech_brake_r {parent=dyn_mech_foot_r, loc=<-4.3392 -0.2253 -0.066>, rot=<0.001 0.0023 -0.0007 -1>}
 dyn_mech_piston01_r {parent=dyn_mech_calf_r, loc=<-2.0967 -0.9094 -0.0191>, rot=<-0.0028 -0.0033 0.0004 -1>}
 dyn_mech_piston02_r {parent=dyn_mech_piston01_r, loc=<-3.2962 0 0>, rot=<0 0 0 1>}
 dyn_mech_pelvis_piston_01_r {parent=mech_spine, loc=<3.0922 -0.8827 0.3059>, rot=<0.5116 -0.3672 -0.4529 -0.6311>}
 dyn_mech_pelvis_piston_02_r {parent=dyn_mech_pelvis_piston_01_r, loc=<-4.3528 0 0>, rot=<0 0 0 -1>}
 dyn_mech_pelvis_piston_01_l {parent=mech_spine, loc=<3.0922 -0.8827 -0.3059>, rot=<-0.3672 -0.5116 0.6311 -0.4529>}
 dyn_mech_pelvis_piston_02_l {parent=dyn_mech_pelvis_piston_01_l, loc=<4.3528 0 0>, rot=<0 0 0 1>}
 mech_chest {parent=body, loc=<0 0 0>, rot=<-0.183 0.683 0.183 -0.683>}
 dyn_mech_neck_01 {parent=mech_chest, loc=<8.7528 3.8186 -0.2952>, rot=<0 0 -0.2588 -0.9659>}
 dyn_mech_neck_02 {parent=dyn_mech_neck_01, loc=<1.2246 0 0>, rot=<0.0001 -0.0001 -0.6093 -0.7929>}
 dyn_mech_neck_03 {parent=dyn_mech_neck_02, loc=<4.1217 0 -0.001>, rot=<0.3413 -0.9399 0 0>}
 dyn_mech_neck04 {parent=dyn_mech_neck_03, loc=<4.1198 0 0>, rot=<0.5963 -0.38 0.5964 0.3799>}
 dyn_mech_head {parent=dyn_mech_neck04, loc=<0.1981 -0.001 0.8267>, rot=<0 0 0 1>}
 dyn_mech_arm {parent=mech_chest, loc=<5.6671 3.8579 -1.8538>, rot=<0.183 -0.683 -0.183 -0.683>}
 dyn_mech_arm_plank {parent=dyn_mech_arm, loc=<1.0396 -0.0022 0.0066>, rot=<0 -0.7071 0 -0.7071>}
 dyn_mech_monitor_rotator {parent=dyn_mech_arm_plank, loc=<3.8195 0 1.4756>, rot=<0.5 -0.5 -0.5 0.5>}
 dyn_mech_monitor_plank {parent=dyn_mech_monitor_rotator, loc=<3.6866 0.006 0.3789>, rot=<0 0 0 1>}
]
