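# Physics asset for the AbstractMirror backpack cosmetic, simulating a small mech figure with simple box, sphere and capsule bodies.
# Arm-to-monitor and neck-to-head chains hang from the chest, and two pelvis-to-toe leg chains hang from the spine.
# Most joints lock swing and keep limited twist as hinges; both planks are fully locked, thighs get slight swing and toes twist freely.
# Pack, chest and spine bodies have no constraints between them.
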PhysicsAsset /BRCosmetics/Accessories/FORT_Backpacks/Backpack_M_MED_AbstractMirror/Meshes/M_MED_AbstractMirror_Pack_Physics_Final
bodies {
 dyn_mech_arm [{shapeType=Sphere, radius=1.8776, shapeName=dyn_mech_arm_sphere}]
 dyn_mech_arm_plank [{shapeType=Box, extentX=8.52, extentY=1.4811, extentZ=0.56, shapeName=dyn_mech_arm_plank_box}]
 dyn_mech_calf_l [{shapeType=Capsule, length=5.3439, radius=1.2096, shapeName=dyn_mech_calf_l_capsule, center=<2.8496 0 0>, rotation=<rot -90 49.9976 -49.9973>}]
 dyn_mech_calf_r [{shapeType=Capsule, length=5.3439, radius=1.2096, shapeName=dyn_mech_calf_l_capsule, center=<-2.8496 0 0>, rotation=<rot -90 179.9996 0.0006>}]
 dyn_mech_foot_l [{shapeType=Capsule, length=4.4239, radius=1.1863, shapeName=dyn_mech_foot_l_capsule, center=<2.021 0 0>, rotation=<rot 90 179.9999 179.9996>}]
 dyn_mech_foot_r [{shapeType=Capsule, length=4.4239, radius=1.1863, shapeName=dyn_mech_foot_l_capsule, center=<-2.021 0 0>, rotation=<rot 90 179.9999 -0.0004>}]
 dyn_mech_head [{shapeType=Capsule, length=5.1262, radius=1.9208, shapeName=dyn_mech_head_capsule, center=<0 0 1.6429>, rotation=<rot 0 0 -90.0002>}]
 dyn_mech_monitor_plank [{shapeType=Box, extentX=14.1385, extentY=1.7608, extentZ=9.0612, shapeName=dyn_mech_monitor_plank_box, center=<1.6256 0 3.7013>}]
 dyn_mech_monitor_rotator [{shapeType=Capsule, length=3.5002, radius=0.7534, shapeName=dyn_mech_monitor_rotator_capsule, center=<1.9768 0 0>, rotation=<rot 90 0 0>}]
 dyn_mech_neck04 [{shapeType=Sphere, radius=0.69, shapeName=dyn_mech_neck04_sphere}]
 dyn_mech_neck_01 [{shapeType=Sphere, radius=0.8994, shapeName=dyn_mech_neck_01_sphere}]
 dyn_mech_neck_02 [{
   shapeType=Capsule
   length=3.8408
   radius=0.5702
   shapeName=dyn_mech_neck_02_capsule
   center=<2.0502 0 0>
   rotation=<rot 90 -166.7373 -166.737>
  }]
 dyn_mech_neck_03 [{shapeType=Capsule, length=4.0869, radius=0.5093, shapeName=dyn_mech_neck_03_capsule, center=<2.1845 0 0>, rotation=<rot -90 -0.0048 0.005>}]
 dyn_mech_pelvis_l [{shapeType=Capsule, length=2.0686, radius=0.5, shapeName=dyn_mech_pelvis_l_capsule, center=<0.791 0 0>, rotation=<rot 90 179.9802 179.9803>}]
 dyn_mech_pelvis_r [{shapeType=Capsule, length=2.0686, radius=0.5, shapeName=dyn_mech_pelvis_l_capsule, center=<-0.791 0 0>, rotation=<rot 90 179.9802 -0.0197>}]
 dyn_mech_thigh_l [{shapeType=Capsule, length=5.6009, radius=1.4306, shapeName=dyn_mech_thigh_l_capsule, center=<3.102 0 0>, rotation=<rot 90 0 -0.0002>}]
 dyn_mech_thigh_r [{shapeType=Capsule, length=5.6009, radius=1.4306, shapeName=dyn_mech_thigh_l_capsule, center=<-3.102 0 0>, rotation=<rot 90 0 179.9998>}]
 dyn_mech_toe_l [{shapeType=Sphere, radius=1.9415, shapeName=dyn_mech_toe_l_sphere}]
 dyn_mech_toe_r [{shapeType=Sphere, radius=1.9415, shapeName=dyn_mech_toe_l_sphere}]
 mech_chest [{shapeType=Capsule, length=9.6168, radius=3.878, shapeName=mech_chest_capsule, center=<5.7134 0.803 0.9339>, rotation=<rot 0 4.9996 90>}]
 mech_spine [{shapeType=Sphere, radius=2.7811, shapeName=mech_spine_sphere}]
 pack [{shapeType=Box, extentX=49.8857, extentY=3.2613, extentZ=24.8932, shapeName=pack_box, center=<0 0.8143 -11.1371>}]
}
constraints [
 {
  bone1Name=dyn_mech_arm
  bone2Name=mech_chest
  swing1LimitDegrees=45
  swing1Motion=Locked
  swing2LimitDegrees=45
  swing2Motion=Locked
  twistLimitDegrees=15
 }
 {
  bone1Name=dyn_mech_arm_plank
  bone2Name=dyn_mech_arm
  swing1LimitDegrees=45
  swing1Motion=Locked
  swing2LimitDegrees=45
  swing2Motion=Locked
  twistLimitDegrees=45
  twistMotion=Locked
 }
 {
  bone1Name=dyn_mech_monitor_rotator
  bone2Name=dyn_mech_arm_plank
  swing1LimitDegrees=45
  swing1Motion=Locked
  swing2LimitDegrees=45
  swing2Motion=Locked
  twistLimitDegrees=5
 }
 {
  bone1Name=dyn_mech_pelvis_r
  bone2Name=mech_spine
  swing1LimitDegrees=0
  swing1Motion=Locked
  swing2LimitDegrees=0
  swing2Motion=Locked
  twistLimitDegrees=5
 }
 {
  bone1Name=dyn_mech_neck_01
  bone2Name=mech_chest
  swing1LimitDegrees=45
  swing1Motion=Locked
  swing2LimitDegrees=45
  swing2Motion=Locked
  twistLimitDegrees=3
 }
 {
  bone1Name=dyn_mech_neck_02
  bone2Name=dyn_mech_neck_01
  swing1LimitDegrees=45
  swing1Motion=Locked
  swing2LimitDegrees=45
  swing2Motion=Locked
  twistLimitDegrees=15
 }
 {
  bone1Name=dyn_mech_neck_03
  bone2Name=dyn_mech_neck_02
  swing1LimitDegrees=0
  swing1Motion=Locked
  swing2LimitDegrees=0
  swing2Motion=Locked
  twistLimitDegrees=15
 }
 {
  bone1Name=dyn_mech_monitor_plank
  bone2Name=dyn_mech_monitor_rotator
  swing1LimitDegrees=45
  swing1Motion=Locked
  swing2LimitDegrees=45
  swing2Motion=Locked
  twistLimitDegrees=45
  twistMotion=Locked
 }
 {
  bone1Name=dyn_mech_head
  bone2Name=dyn_mech_neck04
  swing1LimitDegrees=0
  swing1Motion=Locked
  swing2LimitDegrees=0
  swing2Motion=Locked
  twistLimitDegrees=20
 }
 {
  bone1Name=dyn_mech_neck04
  bone2Name=dyn_mech_neck_03
  swing1LimitDegrees=0
  swing1Motion=Locked
  swing2LimitDegrees=0
  swing2Motion=Locked
  twistLimitDegrees=20
 }
 {
  bone1Name=dyn_mech_pelvis_l
  bone2Name=mech_spine
  swing1LimitDegrees=45
  swing1Motion=Locked
  swing2LimitDegrees=45
  swing2Motion=Locked
  twistLimitDegrees=5
 }
 {bone1Name=dyn_mech_thigh_l, bone2Name=dyn_mech_pelvis_l, swing1LimitDegrees=3, swing2LimitDegrees=3, twistLimitDegrees=3}
 {bone1Name=dyn_mech_thigh_r, bone2Name=dyn_mech_pelvis_r, swing1LimitDegrees=3, swing2LimitDegrees=3, twistLimitDegrees=3}
 {
  bone1Name=dyn_mech_calf_l
  bone2Name=dyn_mech_thigh_l
  swing1LimitDegrees=45
  swing1Motion=Locked
  swing2LimitDegrees=45
  swing2Motion=Locked
  twistLimitDegrees=40
 }
 {
  bone1Name=dyn_mech_calf_r
  bone2Name=dyn_mech_thigh_r
  swing1LimitDegrees=45
  swing1Motion=Locked
  swing2LimitDegrees=45
  swing2Motion=Locked
  twistLimitDegrees=40
 }
 {
  bone1Name=dyn_mech_foot_l
  bone2Name=dyn_mech_calf_l
  swing1LimitDegrees=45
  swing1Motion=Locked
  swing2LimitDegrees=45
  swing2Motion=Locked
  twistLimitDegrees=30
 }
 {
  bone1Name=dyn_mech_foot_r
  bone2Name=dyn_mech_calf_r
  swing1LimitDegrees=45
  swing1Motion=Locked
  swing2LimitDegrees=45
  swing2Motion=Locked
  twistLimitDegrees=30
 }
 {
  bone1Name=dyn_mech_toe_r
  bone2Name=dyn_mech_foot_r
  swing1LimitDegrees=45
  swing1Motion=Locked
  swing2LimitDegrees=45
  swing2Motion=Locked
  twistLimitDegrees=180
  twistMotion=Free
 }
 {
  bone1Name=dyn_mech_toe_l
  bone2Name=dyn_mech_foot_l
  swing1LimitDegrees=45
  swing1Motion=Locked
  swing2LimitDegrees=45
  swing2Motion=Locked
  twistLimitDegrees=45
  twistMotion=Free
 }
]
