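# Animation Blueprint for the AbstractMirror backpack cosmetic, layering idle motion over secondary physics.
# The input pose gets a pack-bone roll during skydive floating, then runs through partial-strength RigidBody and a Control Rig.
# A Random Sequence Player adds weighted idle, idle-break and wave additives on top.
# The "Animation Testing" state machine is unconnected, and EventGraph only sets PlayerEliminated to false, even on the elimination event.
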AnimBlueprint /BRCosmetics/Accessories/FORT_Backpacks/Backpack_M_MED_AbstractMirror/Meshes/M_MED_AbstractMirror_Pack_AnimBP
target_skeleton=&/M_MED_AbstractMirror_Pack_Skeleton
graphs [
 AnimGraph {nodes [
   n1 = "Transform (Modify) Bone - Bone: pack"!(
    ComponentPose=Local To Component(LocalPose=Input Pose())
    bAlphaBoolEnabled:bool=Get bIsSkydiveFloating()
    Translation="0,0,0"
    Rotation="0.000,0.000,15.000"
    Scale
   ) -> (Pose) {
    alphaInputType=Bool
    rotationMode=BMM_Replace
    alphaBoolBlend {blendInTime=0.01, blendOption=Linear, blendOutTime=0.01}
    boneToModify {boneName=pack}
    rotation {roll=15}
   }
   n2 = RigidBody(ComponentPose=@n1, Alpha=0.4) {
    bOverrideWorldGravity=true
    overridePhysicsAsset=&/M_MED_AbstractMirror_Pack_Physics_Final
    alphaBoolBlend {blendOption=Linear}
    componentLinearAccScale=<0.1 0.1 0.1>
    overrideWorldGravity {z=-980}
   }
   n3 = Control Rig(Source=Component To Local(ComponentPose=@n2)) {
    alphaBoolBlend {blendOption=Linear}
    controlRigAssetReference {blueprintRigClass=&/M_MED_AbstractMirror_Pack_CtrlRig.M_MED_AbstractMirror_Pack_CtrlRig_C}
   }
   n4 = Random Sequence Player!() -> (Pose:Pose Link Structure) {
    is=title AnimGraphNode_RandomPlayer
    entries [
     {
      chanceToPlay=0.5
      maxLoopCount=0
      maxPlayRate=1
      minLoopCount=0
      minPlayRate=1
      sequence=/BRCosmetics/Accessories/FORT_Backpacks/Backpack_M_MED_AbstractMirror/Animaiton/AbstractMirror_Idle_PACK
      blendIn {blendOption=Linear, blendTime=0.2, customCurve=null}
     }
     {
      chanceToPlay=0.25
      maxLoopCount=0
      maxPlayRate=1
      minLoopCount=0
      minPlayRate=1
      sequence=/BRCosmetics/Accessories/FORT_Backpacks/Backpack_M_MED_AbstractMirror/Animaiton/AbstractMirror_IdleBreak_PACK
      blendIn {blendOption=Linear, blendTime=0.2, customCurve=null}
     }
     {
      chanceToPlay=0.25
      maxLoopCount=0
      maxPlayRate=1
      minLoopCount=0
      minPlayRate=1
      sequence=/BRCosmetics/Accessories/FORT_Backpacks/Backpack_M_MED_AbstractMirror/Animaiton/AbstractMirror_Wave_PACK
      blendIn {blendOption=Linear, blendTime=0.2, customCurve=null}
     }
    ]
   }
   Output Pose(Result=Apply Additive(Base=@n3, Additive=@n4) {alphaBoolBlend {blendOption=Linear}}) {layerGroup=None}
   AnimGraphNode_StateMachine_0 = Animation Testing() {is=title AnimGraphNode_StateMachine}
  ]}
 Animation Testing {
  entry=State
  path=AnimGraph.AnimGraphNode_StateMachine_0.Animation Testing
  ids=id
  nodes [
   AnimStateNode_0 = State() {
    is=title AnimStateNode
    stateEntered {guid=#1}
    stateFullyBlended {guid=#2}
    stateLeft {guid=#3}
   }
   AnimStateNode_1 = State_2() {
    is=title AnimStateNode
    stateEntered {guid=#4}
    stateFullyBlended {guid=#5}
    stateLeft {guid=#6}
   }
   AnimStateNode_2 = State_1() {
    is=title AnimStateNode
    stateEntered {guid=#7}
    stateFullyBlended {guid=#8}
    stateLeft {guid=#9}
   }
   AnimStateTransitionNode_0 = AnimStateTransitionNode() {
    bAutomaticRuleBasedOnSequencePlayerInState=true
    crossfadeDuration=0
    transitionEnd {guid=#10}
    transitionInterrupt {guid=#11}
    transitionStart {guid=#12}
   }
   AnimStateTransitionNode_1 = AnimStateTransitionNode() {
    bAutomaticRuleBasedOnSequencePlayerInState=true
    crossfadeDuration=0
    transitionEnd {guid=#13}
    transitionInterrupt {guid=#14}
    transitionStart {guid=#15}
   }
   AnimStateTransitionNode_2 = AnimStateTransitionNode() {
    bAutomaticRuleBasedOnSequencePlayerInState=true
    transitionEnd {guid=#16}
    transitionInterrupt {guid=#17}
    transitionStart {guid=#18}
   }
  ]
  links [
   {
    dst=State_1
    rule=AnimGraph.AnimGraphNode_StateMachine_0.Animation Testing.AnimStateTransitionNode_0.Transition
    src=State
    via=AnimStateTransitionNode_0
   }
   {
    dst=State_2
    rule=AnimGraph.AnimGraphNode_StateMachine_0.Animation Testing.AnimStateTransitionNode_1.Transition
    src=State_1
    via=AnimStateTransitionNode_1
   }
   {
    dst=State
    rule=AnimGraph.AnimGraphNode_StateMachine_0.Animation Testing.AnimStateTransitionNode_2.Transition
    src=State_2
    via=AnimStateTransitionNode_2
   }
  ]
 }
 State {
  path=AnimGraph.AnimGraphNode_StateMachine_0.Animation Testing.AnimStateNode_0.State
  nodes [
   Sequence Player 'JollyTroll_Idle_Anim_lookabout_PACK'() {
    bLoopAnimation=false
    sequence=/BRCosmetics/Accessories/FORT_Backpacks/Backpack_M_MED_JollyTroll/Animations/JollyTroll_Idle_Anim_lookabout_PACK
   }
   n1 = "Sequence Player (additive) 'AbstractMirror_Idle_PACK'"() {bLoopAnimation=false, sequence=&AbstractMirror_Idle_PACK}
   Output Animation Pose(Result=@n1)
  ]
 }
 State_2 {
  path=AnimGraph.AnimGraphNode_StateMachine_0.Animation Testing.AnimStateNode_1.State_2
  nodes [
   Sequence Player 'JollyTroll_Sneeze_Anim_PACK'() {bLoopAnimation=false, sequence=/BRCosmetics/Accessories/FORT_Backpacks/Backpack_M_MED_JollyTroll/Animations/JollyTroll_Sneeze_Anim_PACK}
   n1 = "Sequence Player (additive) 'AbstractMirror_IdleBreak_PACK'"() {bLoopAnimation=false, sequence=&AbstractMirror_IdleBreak_PACK}
   Output Animation Pose(Result=@n1)
  ]
 }
 State_1 {
  path=AnimGraph.AnimGraphNode_StateMachine_0.Animation Testing.AnimStateNode_2.State_1
  nodes [
   Sequence Player 'JollyTroll_Sneeze_Anim_PACK'() {bLoopAnimation=false, sequence=&JollyTroll_Sneeze_Anim_PACK}
   n1 = "Sequence Player (additive) 'AbstractMirror_Wave_PACK'"() {bLoopAnimation=false, sequence=&AbstractMirror_Wave_PACK}
   Output Animation Pose(Result=@n1)
  ]
 }
 Transition {path=AnimGraph.AnimGraphNode_StateMachine_0.Animation Testing.AnimStateTransitionNode_0.Transition, nodes [Result()]}
 Transition {path=AnimGraph.AnimGraphNode_StateMachine_0.Animation Testing.AnimStateTransitionNode_1.Transition, nodes [Result()]}
 Transition {path=AnimGraph.AnimGraphNode_StateMachine_0.Animation Testing.AnimStateTransitionNode_2.Transition, nodes [Result()]}
 EventGraph {nodes [
   n1 = Event BlueprintProcessAnimInputEvent()
   > Branch(Condition=IsNamed(self=@n1.AnimInputEvent, InEventName=aie.EnemyPlayerEliminated))
   > Set PlayerEliminated(PlayerEliminated:bool=false) -> (Output_Get:bool)
   Event BlueprintPostEvaluateAnimation() {is=title K2Node_Event}
   > Set PlayerEliminated(PlayerEliminated=false)
   TryGetPawnOwner()
   Event BlueprintUpdateAnimation()
  ]}
]
variables [PlayerEliminated {type=boolean}]
